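# Masked, lit smoke-ribbon trail material for the Durrburger glider, usable on Niagara ribbons and beam trails.
# Smoke textures pan along U-stretched, tiled UVs, and their normals blend with a tiling detail normal.
# Emissive combines low/high color tinting and a fake sun/moon reflection highlight, scaled by time of day and camera vertical angle.
# Dynamic parameters drive erosion of the mask, which sphere masks fade toward the ribbon's edges.
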Material /Game/Gadgets/Assets/VinderTech_GliderChute/Glider_Durrburger_Hero/FX/M_Glider_Durrburger_Trail
graph {
 properties {
  bUsedWithBeamTrails=true
  bUsedWithMeshDeformer=true
  bUsedWithNiagaraRibbons=true
  bUsedWithSkeletalMesh=true
  blendMode=BLEND_Masked
  pixelDepthOffsetMode=PDOM_Legacy
  usedShadingModels=Default Lit
 }
 nodes [
  Constant() {r=1}
  n1 = MaterialFunctionCall(Float2=StaticSwitchParameter(True=TextureCoordinate() {coordinateIndex=1}, False=TextureCoordinate()) {parameterName="TexCoord (1)"}) {materialFunction=&BreakOutFloat2Components}
  n2 = Multiply(A=@n1.R, B=ScalarParameter() {defaultValue=0.7, parameterName=Utile})
  n3 = ScalarParameter() {defaultValue=2, parameterName=UstartStretch}
  n4 = Panner(
   Coordinate=AppendVector(A=Power(Base=@n2, Exp=@n3), B=@n1.G)
   Time=Add(A=Time(), B=ScalarParameter() {parameterName=TimeOffset})
   Speed=AppendVector(A=ScalarParameter() {defaultValue=-1, parameterName=Speed}, B=Constant())
  ) {speedX=-9}
  n5 = MaterialFunctionCall(
   BaseNormal=TextureSample(UVs=@n4) {samplerType=SAMPLERTYPE_Normal, texture=/Game/Effects/Fort_Effects/Gameplay/SmokeDustExplosions/Textures/T_RibbonSmokeMasked_01}.RGB
   AdditionalNormal=TextureSample(UVs=TextureCoordinate() {uTiling=10, vTiling=10}) {samplerType=SAMPLERTYPE_Normal, texture=/Game/Effects/Fort_Effects/Textures/Husk/Poison/T_PoisonTile_N}.RGB
  ) {materialFunction=&BlendAngleCorrectedNormals}
  LinearInterpolate(A=Constant3Vector() {constant=<rgba 0 0 1 0>}, B=@n5.Result, Alpha=ScalarParameter() {defaultValue=1, parameterName=NormalIntensity})
  n6 = CameraVectorWS()
  n7 = Saturate(None=Divide(A=ComponentMask(None=@n6) {b=true}) {constB=20})
  n8 = LinearInterpolate(Alpha=@n7) {constA=0.7, constB=15}
  n9 = ScalarParameter() {defaultValue=0.05, parameterName=SubsurfaceDownsideIntensity}
  n10 = ScalarParameter() {defaultValue=20, parameterName=SubsurfaceUpsideIntensity}
  n11 = MaterialFunctionCall(
   Color_Morning=@n8
   Color_Day=@n8
   Color_Evening=@n8
   Color_Night=Multiply(
    A=LinearInterpolate(A=@n9, B=@n10, Alpha=@n7) {constA=0.05, constB=20}
    B=VectorParameter() {parameterName=NightColorTweak, defaultValue=<rgba 1 1 1 0>}.RGB
   )
  ) {materialFunction=&Color_Time_of_Day}
  n12 = CollectionParameter() {
   collection=&FortniteMaterialParameters
   parameterName=SunAndMoonModelDirectionalVector
  }
  n13 = DotProduct(A=ComponentMask(None=@n12) {b=true, g=true, r=true}, B=ReflectionVectorWS(CustomWorldNormal=@n5.Result) {bNormalizeCustomWorldNormal=true})
  n14 = ScalarParameter() {defaultValue=1, parameterName=FakeSpecularPower}
  n15 = TextureSample(UVs=@n4) {samplerType=SAMPLERTYPE_LinearColor, texture=/Game/Effects/Fort_Effects/Gameplay/SmokeDustExplosions/Textures/T_RibbonSmokeMasked_03}
  n16 = MaterialFunctionCall!(
   Blend=Multiply(A=Power(Base=@n13, Exp=@n14) {constExponent=6}, B=VectorParameter() {parameterName=FakeSpecularColor, defaultValue=<rgba 2 0.9024 0 1>}.RGB)
   Base=LinearInterpolate(
    A=VectorParameter() {parameterName=LowColor, defaultValue=<rgba 0.4219 0.1797 0.0109 1>}.RGB
    B=VectorParameter() {parameterName=HighColor, defaultValue=<rgba 1 0.5767 0.0724 1>}.RGB
    Alpha=LinearInterpolate(Alpha=@n15.R) {constA=0.1}
   )
  ) {materialFunction=/Engine/Functions/Engine_MaterialFunctions03/Blends/Blend_Lighten}
  MaterialFunctionCall(Emissive=Multiply(A=Multiply(A=@n11.Result, B=@n16.Result), B=ScalarParameter() {defaultValue=1, parameterName=EmissiveAmount, sortPriority=1})) {materialFunction=&MF_FX_EmissiveAdjustments}
  n17 = TextureSample(UVs=@n4) {texture=/Game/Effects/Fort_Effects/Gameplay/SmokeDustExplosions/Textures/T_RibbonSmokeMasked_02}
  n18 = OneMinus(None=@n15.R)
  n19 = MaterialFunctionCall(Float2=TextureCoordinate()) {materialFunction=&BreakOutFloat2Components}
  n20 = DynamicParameter() {paramNames [Erode,UmaskExp,Param3,ErodeByBP]}
  n21 = Saturate(None=Subtract(A=Power(Base=@n19.R, Exp=@n20.UmaskExp)) {constB=0.05})
  n22 = Multiply(A=Multiply(A=Multiply(A=@n18, B=@n21), B=@n20.Erode) {constB=2}, B=@n20.ErodeByBP) {constB=2}
  n23 = TextureCoordinate()
  n24 = Constant2Vector() {g=0.5}
  n25 = Constant() {r=1}
  Multiply(
   A=Multiply(A=Subtract(A=@n17.R, B=@n22), B=SphereMask(A=@n23, B=@n24, Radius=@n25) {hardnessPercent=0})
   B=SphereMask(A=TextureCoordinate(), B=Constant2Vector() {g=0.5, r=1}, Radius=Constant() {r=1})
  )
  Multiply(A=@n15.R, B=ScalarParameter() {defaultValue=1, parameterName=Specular})
  ScalarParameter() {defaultValue=0.3, parameterName=Roughness}
 ]
}
